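# Niagara glow effect for the Hurtle emote, with burst-spawned particles that persist instead of expiring after their lifetime.
# LightDecal renders a mesh in the system glow color, offset downward and faded in quickly by emitter age.
# LightDecal001 spawns four light-and-mesh particles, using particle ID to pick fixed positions and radii, with radius ramping up slightly later.
# A system spawn assignment and a color-array selector support the shared color variables.
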NiagaraSystem /BRCosmetics/Effects/Fort_Effects/Emotes/Hurtle/NS_Emote_Hurtle_Glow
dependencies {
 usedDataInterfaces [
  &NiagaraDataInterfaceCurve
  &NiagaraDataInterfaceArrayFloat3
  &NiagaraDataInterfaceArrayFloat
  &NiagaraDataInterfaceArrayColor
 ]
 usedDynamicInputs [&SelectLinearColorFromArray]
 usedModules [
  &SystemState
  &self:SystemSpawnScript.SystemScriptSource.SystemScriptGraph.NiagaraNodeAssignment_0.SetVariables_#1
 ]
}
emitters {
 LightDecal {
  emitterUpdateScript [
   &EmitterState()
   &SpawnBurst_Instantaneous(Use Spawn Probability=false, UseLoopCountLimit=true)
  ]
  particleSpawnScript [&InitializeParticle(
    Color=$System.Glow Color
    Lifetime=5
    Mesh Scale Mode=::NewEnumerator3
    Position Offset=<0 0 -25>
    UsePositionOffset=true
    Write Material Random=true
   )]
  particleUpdateScript [
   &ParticleState(Kill Particles When Lifetime Has Elapsed=false, Let Infinitely Lived Particles Die When Emitter Deactivates=true)
   &ScaleColor(
    Scale RGB=&VectorFromFloat(Value=&FloatFromCurve(
      FloatCurve=DataInterface() {propertyValues {Curve {keys [
          {arriveTangent=18.7864, interpMode=RCIM_Cubic, leaveTangent=18.7864, tangentMode=RCTM_User, time=0, value=0}
          {interpMode=RCIM_Cubic, tangentMode=RCTM_User, time=0.1, value=1}
         ]}}}
      CurveIndex=$Emitter.Age
     ))
    ScaleA=true
    ScaleRGB=true
   )
  ]
  rendererClasses [&NiagaraMeshRendererProperties]
  renderers [{rendererClass=&NiagaraMeshRendererProperties, rendererIndex=0}]
 }
 LightDecal001 {
  emitterUpdateScript [
   &EmitterState()
   &SpawnBurst_Instantaneous(Spawn Count=4, Use Spawn Probability=false, UseLoopCountLimit=true)
  ]
  particleSpawnScript [&InitializeParticle(Lifetime=5, Mesh Scale Mode=::NewEnumerator3, Mesh Uniform Scale=10, UsePositionOffset=false, Write Material Random=true)]
  particleUpdateScript [
   &ParticleState(Kill Particles When Lifetime Has Elapsed=false, Let Infinitely Lived Particles Die When Emitter Deactivates=true)
   &FN_Light_Attributes(
    Color=$System.Light Color
    Create Color Attribute=true
    Create Exponent Attribute=true
    Create Position Attribute=true
    Create Radius Attribute=true
    Exponent=8
    Position=&SelectVectorFromArray(
     Direct Array Index=&ReturnParticleIDAttributes()
     Vector Selection Array=DataInterface() {propertyValues {FloatData [<0 92 -20>, <0 -80 -20>, <45 0 -20>, <-45 0 -20>]}}
    )
    Radius=&Multiply_Float(
     A=&SelectFloatFromArray(
      Direct Array Index=&ReturnParticleIDAttributes()
      Float Selection Array=DataInterface() {propertyValues {FloatData [100,60,60,60]}}
     )
     B=&FloatFromCurve(
      FloatCurve=DataInterface() {propertyValues {Curve {keys [
          {arriveTangent=51.8489, interpMode=RCIM_Cubic, leaveTangent=51.8489, tangentMode=RCTM_User, time=0.35, value=0}
          {interpMode=RCIM_Cubic, tangentMode=RCTM_User, time=0.4, value=1}
         ]}}}
      CurveIndex=$Emitter.Age
     )
    )
    Scale Color=1.5
   )
  ]
  rendererClasses [&NiagaraLightRendererProperties, &NiagaraMeshRendererProperties]
  renderers [
   {rendererClass=&NiagaraLightRendererProperties, rendererIndex=0}
   {rendererClass=&NiagaraMeshRendererProperties, rendererIndex=1}
  ]
 }
}
summary {
 systemName=NS_Emote_Hurtle_Glow
 emitters [
  {emitterName=LightDecal, rendererClasses [&NiagaraMeshRendererProperties]}
  {emitterName=LightDecal001, rendererClasses [&NiagaraLightRendererProperties, &NiagaraMeshRendererProperties]}
 ]
 userVariables [
  User.bIsFrontEnd {type=NiagaraBool}
  User.bIsFrontEndPreview {type=NiagaraBool}
  User.bIsInWater {type=NiagaraBool}
  User.Color {type=NiagaraInt32}
 ]
}
