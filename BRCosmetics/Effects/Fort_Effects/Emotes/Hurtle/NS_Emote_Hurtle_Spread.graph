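# Niagara system for the Hurtle emote that sprays randomly tinted V-Buck meshes from three emitters.
# Top_Vbuck flickers short-lived, flattened meshes at the source, while Vbucks streams meshes along +Y, with User.Rate scaling spawn rate and launch speed.
# Vbucks_Burst adds a delayed one-shot spray.
# Flying meshes tumble under curl noise, gravity and drag, thicken then shrink with age, and follow source movement with a fading weight.
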NiagaraSystem /BRCosmetics/Effects/Fort_Effects/Emotes/Hurtle/NS_Emote_Hurtle_Spread
dependencies {
 usedDataInterfaces [
  &NiagaraDataInterfaceArrayColor
  &NiagaraDataInterfaceCurve
  &NiagaraDataInterfaceMeshRendererInfo
  &NiagaraDataInterfaceVectorCurve
 ]
 usedDynamicInputs [&Multiply_Vector2D, &MakeVector2D]
 usedModules [&SystemState]
}
emitters {
 Top_Vbuck {
  emitterUpdateScript [
   &EmitterState(Life Cycle Mode=::NewEnumerator1, Loop Behavior=::NewEnumerator1, Loop Duration=2, UseLoopDelay=false)
   &SpawnRate(SpawnRate=8, Use Spawn Probability=false)
  ]
  particleSpawnScript [
   &InitializeParticle(
    Color=&SelectLinearColorFromArray(
     Array Sampling Mode=::NewEnumerator0
     Direct Array Index=?
     Color Selection Array=DataInterface() {propertyValues {ColorData [<rgba 0.9 0.81 0.225 1>, <rgba 1 0.45 0.5875 1>, <rgba 0.36 0.9 0.5535 1>, <rgba 0.8019 0.5 1 1>, <rgba 1 0.5127 0.3 1>]}}
     Random Seed=0
    )
    Lifetime=0.1
    Mesh Scale=<1 1 0.01>
    Mesh Scale Mode=::NewEnumerator4
    Position Offset=<0 0 2.75>
    UsePositionOffset=true
    Write Material Random=true
   )
   &InitialMeshOrientation(Mesh Orientation Mode=::NewEnumerator1, Rotation=<0 0 0.25>, Use Rotation Vector=true)
   &CameraOffset(Camera Offset Amount=1)
  ]
  particleUpdateScript [&ParticleState()]
  rendererClasses [&NiagaraMeshRendererProperties]
  renderers [{rendererClass=&NiagaraMeshRendererProperties, rendererIndex=0}]
 }
 Vbucks {
  emitterUpdateScript [
   &EmitterState(Life Cycle Mode=::NewEnumerator1, Loop Behavior=::NewEnumerator1, Loop Duration=2, UseLoopDelay=true)
   &SpawnRate(SpawnRate=&Lerp_Float(A=8, B=18, Alpha=$User.Rate), Use Spawn Probability=false)
  ]
  particleSpawnScript [
   &InitializeParticle(
    Color=&SelectLinearColorFromArray(
     Array Sampling Mode=::NewEnumerator0
     Direct Array Index=?
     Color Selection Array=DataInterface() {propertyValues {ColorData [<rgba 0.9 0.81 0.225 1>, <rgba 1 0.45 0.5875 1>, <rgba 0.36 0.9 0.5535 1>, <rgba 0.8019 0.5 1 1>, <rgba 1 0.5127 0.3 1>]}}
     Random Seed=0
    )
    Lifetime Max=1.2
    Lifetime Min=0.85
    Lifetime Mode=::NewEnumerator1
    Mesh Renderer Array Visibility Mode=::NewEnumerator3
    Mesh Renderer Info=DataInterface() {propertyValues {MeshRenderer=&self:NiagaraEmitter_0.NiagaraMeshRendererProperties_1}}
    Mesh Scale Mode=::NewEnumerator3
    Position Offset=<0 0 1.5>
    UsePositionOffset=true
    Write Material Random=true
   )
   &InitialMeshOrientation(Mesh Orientation Mode=::NewEnumerator1, Rotation=<0 0 0.25>, Use Rotation Vector=true)
   &AddVelocity(
    Velocity=&RandomRangeVector(Minimum=<0 350 20>, Maximum=<0 450 30>)
    Velocity Speed Scale=&Lerp_Float(A=0.5, B=1.2, Alpha=$User.Rate)
   )
   &AddVelocity(
    Cone Angle=90
    Cone Axis=<0 1 0>
    Inner Cone Angle=45
    Module.UseConeFalloff=false
    Module.UseVelDistribution=false
    Override Seed=false
    Velocity Mode=::NewEnumerator2
    Velocity Speed=50
   ) {module=AddVelocity001}
   &ApplyInitialForces()
   &CameraOffset(Camera Offset Amount=2)
  ]
  particleUpdateScript [
   &ParticleState()
   &ScaleMeshSize(Scale Factor=&Multiply_VectorByFloat(
     Vector=&VectorFromCurve(
      VectorCurve=DataInterface() {propertyValues {XCurve {keys [<at 0 1>,<at 1 1>]}, YCurve {keys [<at 0 1>,<at 1 1>]}, ZCurve {keys [<at 0 0.1>,<at 0.4 1>]}}}
      CurveIndex=$Particles.Age
     )
     Float=&FloatFromCurve(FloatCurve=DataInterface() {propertyValues {Curve {keys [{tangentMode=RCTM_User, time=0.7, value=1}, {tangentMode=RCTM_User, time=1, value=0}]}}})
    ))
   &UpdateMeshOrientation(
    Rotation Rate=&RandomRangeFloat(Minimum=-1, Maximum=1)
    Rotation Vector=&Multiply_VectorByFloat(
     Vector=&RandomRangeVector(Minimum=<0 0 0>, Maximum=<1 1 1>)
     Float=&FloatFromCurve(FloatCurve=DataInterface() {propertyValues {Curve {keys [<at 0 0 RCIM_Cubic>,<at 0.1 1 RCIM_Cubic>]}}})
    )
   )
   &AerodynamicDrag(Ignore Mass=true, Surface Normal Axis:Vector3f=<0 0 1>, Use Linear Drag=true, Use Rotational Drag=true)
   &InheritSourceMovement(
    Applied Position Scale=&VectorFromFloat(Value=&FloatFromCurve(FloatCurve=DataInterface() {propertyValues {Curve {keys [
          {arriveTangent=-2.1575, interpMode=RCIM_Cubic, leaveTangent=-2.1575, tangentMode=RCTM_User, time=0.05, value=1}
          {interpMode=RCIM_Cubic, tangentMode=RCTM_User, time=1, value=0}
         ]}}}))
    ApplyForce=false
    ApplyPosition=true
    ApplyPositionFalloff=false
   )
   /Niagara/Modules/Update/Forces/CurlNoiseForce(Noise Frequency=0.03, Noise Quality / Cost=::NewEnumerator0, Noise Strength=333, Pan Noise=true, Pan Noise Field=<0 0 1>)
   &GravityForce()
   &SolveForcesAndVelocity(Clamp Velocity=false, Limit Acceleration=false)
   &AlignSpriteToMeshOrientation(SetAlignment=true, SetFacing=true)
  ]
  rendererClasses [&NiagaraMeshRendererProperties]
  renderers [{rendererClass=&NiagaraMeshRendererProperties, rendererIndex=0}]
 }
 Vbucks_Burst {
  emitterUpdateScript [
   &EmitterState()
   &SpawnBurst_Instantaneous(
    Spawn Count=&RandomRangeInt(Minimum=13, Maximum=17)
    Spawn Time=2
    Use Spawn Probability=false
    UseLoopCountLimit=false
   )
  ]
  particleSpawnScript [
   &InitializeParticle(
    Color=&SelectLinearColorFromArray(
     Array Sampling Mode=::NewEnumerator0
     Direct Array Index=?
     Color Selection Array=DataInterface() {propertyValues {ColorData [<rgba 0.9 0.81 0.225 1>, <rgba 1 0.45 0.5875 1>, <rgba 0.36 0.9 0.5535 1>, <rgba 0.8019 0.5 1 1>, <rgba 1 0.5127 0.3 1>]}}
     Random Seed=0
    )
    Lifetime=1.75
    Mass Max=1.05
    Mass Min=0.9
    Mass Mode=::NewEnumerator3
    Mesh Renderer Array Visibility Mode=::NewEnumerator3
    Mesh Renderer Info=DataInterface() {propertyValues {MeshRenderer=&self:Vbucks_Burst_3.NiagaraMeshRendererProperties_1}}
    Mesh Scale Mode=::NewEnumerator3
    UsePositionOffset=false
    Write Material Random=true
   )
   &InitialMeshOrientation(Use Rotation Vector=true)
   &ShapeLocation(MaskSpawnGroup=false, Override Seed=false, Sphere Radius=2, UseOwnerScale=true)
   &InheritVelocity(Inherited Velocity Amount Scale=<1 1 1.5>, Inherited Velocity Speed Limit=450, Speed Limit=true, UseSpeedThreshold=false) {enabled=false}
   &AddVelocity(Velocity=<0 450 50>)
   &AddVelocity(
    Cone Axis=<0 1 1>
    Inner Cone Angle=22.5
    Module.UseConeFalloff=false
    Module.UseVelDistribution=false
    Override Seed=false
    Velocity Mode=::NewEnumerator2
    Velocity Speed=50
   ) {module=AddVelocity001}
   &ApplyInitialForces()
  ]
  particleUpdateScript [
   &ParticleState()
   &ScaleMeshSize(Scale Factor=&Multiply_VectorByFloat(
     Vector=&VectorFromCurve(
      VectorCurve=DataInterface() {propertyValues {XCurve {keys [<at 0 1>,<at 1 1>]}, YCurve {keys [<at 0 1>,<at 1 1>]}, ZCurve {keys [<at 0 0.1>,<at 0.6 1>]}}}
      CurveIndex=$Particles.Age
     )
     Float=&FloatFromCurve(FloatCurve=DataInterface() {propertyValues {Curve {keys [{tangentMode=RCTM_User, time=0.7, value=1}, {tangentMode=RCTM_User, time=1, value=0}]}}})
    ))
   &UpdateMeshOrientation(
    Rotation Rate=&RandomRangeFloat(Minimum=-1, Maximum=1)
    Rotation Vector=&Multiply_VectorByFloat(
     Vector=&RandomRangeVector(Minimum=<0 0 0>, Maximum=<1 1 1>)
     Float=&FloatFromCurve(FloatCurve=DataInterface() {propertyValues {Curve {keys [<at 0 0 RCIM_Cubic>,<at 0.1 1 RCIM_Cubic>]}}})
    )
   )
   &InheritSourceMovement(
    Applied Position Scale=&VectorFromFloat(Value=&FloatFromCurve(FloatCurve=DataInterface() {propertyValues {Curve {keys [
          {arriveTangent=-2.1575, interpMode=RCIM_Cubic, leaveTangent=-2.1575, tangentMode=RCTM_User, time=0.05, value=1}
          {interpMode=RCIM_Cubic, tangentMode=RCTM_User, time=1, value=0}
         ]}}}))
    ApplyForce=false
    ApplyPosition=true
    ApplyPositionFalloff=false
   )
   &CurlNoiseForce(Noise Frequency=4, Noise Quality / Cost=::NewEnumerator0, Noise Strength=250, Pan Noise=true, Pan Noise Field=<0 0 1>)
   &GravityForce()
   &AerodynamicDrag(
    Aerodynamic Drag:NiagaraFloat=&FloatFromCurve(
     FloatCurve=DataInterface() {propertyValues {Curve {keys [{tangentMode=RCTM_User, time=0, value=0.35}, {tangentMode=RCTM_User, time=0.5, value=1}]}}}
     Scale Curve=&RandomRangeFloat(Minimum=1, Maximum=1.5)
    )
    Aerodynamic Rotational Drag:NiagaraFloat=&RandomRangeFloat(Minimum=0.75, Maximum=1.5)
    Ignore Mass=true
    Surface Normal Axis:Vector3f=<0 0 1>
    Use Linear Drag=true
    Use Rotational Drag=true
   )
   &SolveForcesAndVelocity(Clamp Velocity=false, Limit Acceleration=false)
   &AlignSpriteToMeshOrientation(SetAlignment=true, SetFacing=true)
  ]
  rendererClasses [&NiagaraMeshRendererProperties]
  renderers [{rendererClass=&NiagaraMeshRendererProperties, rendererIndex=0}]
 }
}
summary {
 systemName=NS_Emote_Hurtle_Spread
 emitters [
  {emitterName=Vbucks_Burst, rendererClasses [&NiagaraMeshRendererProperties]}
  {emitterName=Vbucks, rendererClasses [&NiagaraMeshRendererProperties]}
  {emitterName=Top_Vbuck, rendererClasses [&NiagaraMeshRendererProperties]}
 ]
 userVariables [User.Rate {type=NiagaraFloat, defaultValue=0, description=""}]
}
